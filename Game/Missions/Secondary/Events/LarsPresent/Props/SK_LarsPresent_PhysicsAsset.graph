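# Physics asset for the LarsPresent skeletal mesh, a prop in a secondary mission event.
# It defines six bodies: the root, four base pieces named for the positive and negative X and Y sides, and a Z-positive top, which suggests a box that opens.
# No constraints link the bodies, so the asset sets up no joints between the pieces.
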PhysicsAsset /Game/Missions/Secondary/Events/LarsPresent/Props/SK_LarsPresent_PhysicsAsset
bodies {root [], x_neg_base [], x_pos_base [], y_neg_base [], y_pos_base [], z_pos_top []}
constraints []
